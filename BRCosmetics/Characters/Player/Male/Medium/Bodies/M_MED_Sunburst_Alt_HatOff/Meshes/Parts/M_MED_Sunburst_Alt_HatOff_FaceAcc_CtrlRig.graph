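# Control Rig for the Sunburst Alt HatOff face accessory that applies corrective offsets to back and side hair bones as the head moves.
# A spherical pose reader measures the head against spine_05, with debug drawing enabled, and weights eight bone adjustments.
# The shared "Set Transforms for Bone" function adds translate, Euler rotate and scale offsets to each current local transform.
# A second, full-weight chain for the same bones is not connected to execution.
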ControlRigBlueprint /BRCosmetics/Characters/Player/Male/Medium/Bodies/M_MED_Sunburst_Alt_HatOff/Meshes/Parts/M_MED_Sunburst_Alt_HatOff_FaceAcc_CtrlRig
graphs [
 ContainedGraph {nodes [
   n1 = RigVMFunctionEntryNode()
   RigUnit_GetTransform = &RigUnit_GetTransform(Space=LocalSpace, bInitial=false, Item=@n1.Joint)
   n2 = &RigVMFunction_MathQuaternionToEuler(RotationOrder=ZYX, Value=@RigUnit_GetTransform.Transform.Rotation)
   > RigUnit_SetTransform = &RigUnit_SetTransform(
    Space=LocalSpace
    Value=(Rotation=(X=0.0132,Y=0.4099,Z=-0.0222,W=0.9118),Translation=(X=23.6045,Y=-6.37,Z=140.1115),Scale3D=(X=1,Y=1,Z=1))
    bInitial=false
    Item=@n1.Joint
    Weight=@n1.Weight
    Value.Translation.Z=&RigVMFunction_MathFloatAdd(B=@n1.Translate Z, A=@RigUnit_GetTransform.Transform.Translation.Z) {template="Add::Execute(in A,in B,out Result)"}.Result
    Value.Scale3D.Z=&RigVMFunction_MathFloatAdd(B=@n1.Scale Z, A=@RigUnit_GetTransform.Transform.Scale3D.Z) {template="Add::Execute(in A,in B,out Result)"}.Result
    Value.Translation.Y=&RigVMFunction_MathFloatAdd(B=@n1.Translate Y, A=@RigUnit_GetTransform.Transform.Translation.Y) {template="Add::Execute(in A,in B,out Result)"}.Result
    Value.Scale3D.Y=&RigVMFunction_MathFloatAdd(B=@n1.Scale Y, A=@RigUnit_GetTransform.Transform.Scale3D.Y) {template="Add::Execute(in A,in B,out Result)"}.Result
    Value.Translation.X=&RigVMFunction_MathFloatAdd(B=@n1.Translate X, A=@RigUnit_GetTransform.Transform.Translation.X) {template="Add::Execute(in A,in B,out Result)"}.Result
    Value.Scale3D.X=&RigVMFunction_MathFloatAdd(B=@n1.Scale X, A=@RigUnit_GetTransform.Transform.Scale3D.X) {template="Add::Execute(in A,in B,out Result)"}.Result
    Value.Rotation=&RigVMFunction_MathQuaternionFromEuler(
     Euler=(X=0,Y=0,Z=0)
     Euler.X=&RigVMFunction_MathFloatAdd(B=@n1.Rotate X, A=@n2.Result.X) {template="Add::Execute(in A,in B,out Result)"}.Result
     Euler.Y=&RigVMFunction_MathFloatAdd(B=@n1.Rotate Y, A=@n2.Result.Y) {template="Add::Execute(in A,in B,out Result)"}.Result
     Euler.Z=&RigVMFunction_MathFloatAdd(B=@n1.Rotate Z, A=@n2.Result.Z) {template="Add::Execute(in A,in B,out Result)"}.Result
    ).Result
   )
   > RigVMFunctionReturnNode()
  ]}
 RigVMFunctionLibrary {nodes [Set Transforms for Bone = RigVMCollapseNode(
    ExecuteContext=()
    Joint=()
    Rotate X=0
    Rotate Y=0
    Rotate Z=0
    Scale X=0
    Scale Y=0
    Scale Z=0
    Translate X=0
    Translate Y=0
    Translate Z=0
    Weight=1
   )]}
 RigVMModel {nodes [
   RigUnit_BeginExecution = &RigUnit_BeginExecution()
   n1 = &RigUnit_SphericalPoseReader(
    ActiveRegionScaleFactors=(PositiveWidth=1,NegativeWidth=1,PositiveHeight=1,NegativeHeight=1)
    ActiveRegionSize=0.2
    Debug=(bDrawDebug=True,DebugScale=25,DebugSegments=20,DebugThickness=0.25)
    DriverAxis=(X=-1,Y=0,Z=0)
    DriverItem=(Type=Bone,Name="head")
    FalloffRegionScaleFactors=(PositiveWidth=1,NegativeWidth=1,PositiveHeight=1,NegativeHeight=1)
    FalloffSize=0.4
    OptionalParentItem=(Type=Bone,Name="spine_05")
    RotationOffset=(X=90,Y=0,Z=40)
    ExecutePin=@RigUnit_BeginExecution.ExecutePin
   )
   > Set Transforms for Bone_3 = RigVMFunctionReferenceNode(
    Joint=(Type=Bone,Name="hair_bl_mstr")
    Rotate X=0
    Rotate Y=-10
    Rotate Z=0
    Scale X=-0.1
    Scale Y=0
    Scale Z=0
    Translate X=0
    Translate Y=0
    Translate Z=0
    Weight=@n1.OutputParam
   )
   > Set Transforms for Bone_3_2 = RigVMFunctionReferenceNode(
    Joint=(Type=Bone,Name="hair_bl_02_dyn")
    Rotate X=0
    Rotate Y=5
    Rotate Z=0
    Scale X=0
    Scale Y=0
    Scale Z=0
    Translate X=0
    Translate Y=0
    Translate Z=0
    Weight=@n1.OutputParam
   )
   > Set Transforms for Bone_3_1 = RigVMFunctionReferenceNode(
    Joint=(Type=Bone,Name="hair_br_mstr")
    Rotate X=0
    Rotate Y=-10
    Rotate Z=0
    Scale X=-0.1
    Scale Y=0
    Scale Z=0
    Translate X=0
    Translate Y=3
    Translate Z=0
    Weight=@n1.OutputParam
   )
   > Set Transforms for Bone_3_1_1 = RigVMFunctionReferenceNode(
    Joint=(Type=Bone,Name="hair_br_02_dyn")
    Rotate X=0
    Rotate Y=5
    Rotate Z=0
    Scale X=0
    Scale Y=0
    Scale Z=0
    Translate X=0
    Translate Y=0
    Translate Z=0
    Weight=@n1.OutputParam
   )
   > Set Transforms for Bone_3_1_1_1 = RigVMFunctionReferenceNode(
    Joint=(Type=Bone,Name="hair_mid_bl_mstr")
    Rotate X=0
    Rotate Y=0
    Rotate Z=5
    Scale X=0
    Scale Y=0
    Scale Z=0
    Translate X=0
    Translate Y=0
    Translate Z=3
    Weight=@n1.OutputParam
   )
   > Set Transforms for Bone_3_1_1_1_1 = RigVMFunctionReferenceNode(
    Joint=(Type=Bone,Name="hair_mid_br_mstr")
    Rotate X=0
    Rotate Y=0
    Rotate Z=-5
    Scale X=0
    Scale Y=0
    Scale Z=0
    Translate X=0
    Translate Y=7
    Translate Z=0
    Weight=@n1.OutputParam
   )
   > Set Transforms for Bone_3_1_1_1_2 = RigVMFunctionReferenceNode(
    Joint=(Type=Bone,Name="hair_mid_fl_mstr")
    Rotate X=0
    Rotate Y=0
    Rotate Z=5
    Scale X=0
    Scale Y=0
    Scale Z=0
    Translate X=0
    Translate Y=0
    Translate Z=3
    Weight=@n1.OutputParam
   )
   > Set Transforms for Bone_3_1_1_1_1_1 = RigVMFunctionReferenceNode(
    Joint=(Type=Bone,Name="hair_mid_fr_mstr")
    Rotate X=0
    Rotate Y=0
    Rotate Z=-5
    Scale X=0
    Scale Y=0
    Scale Z=0
    Translate X=0
    Translate Y=0
    Translate Z=0
    Weight=@n1.OutputParam
   )
   Set Transforms for Bone_3_3 = RigVMFunctionReferenceNode(
    ExecuteContext=()
    Joint=(Type=Bone,Name="hair_bl_mstr")
    Rotate X=0
    Rotate Y=-15
    Rotate Z=0
    Scale X=-0.2
    Scale Y=0
    Scale Z=0
    Translate X=0
    Translate Y=0
    Translate Z=3
    Weight=1
   )
   > Set Transforms for Bone_3_2_1 = RigVMFunctionReferenceNode(
    Joint=(Type=Bone,Name="hair_bl_02_dyn")
    Rotate X=0
    Rotate Y=15
    Rotate Z=0
    Scale X=0
    Scale Y=0
    Scale Z=0
    Translate X=0
    Translate Y=0
    Translate Z=0
    Weight=1
   )
   > Set Transforms for Bone_3_1_2 = RigVMFunctionReferenceNode(
    Joint=(Type=Bone,Name="hair_br_mstr")
    Rotate X=0
    Rotate Y=-45
    Rotate Z=0
    Scale X=-0.2
    Scale Y=0
    Scale Z=0
    Translate X=0
    Translate Y=3
    Translate Z=0
    Weight=1
   )
   > Set Transforms for Bone_3_1_1_2 = RigVMFunctionReferenceNode(
    Joint=(Type=Bone,Name="hair_br_02_dyn")
    Rotate X=0
    Rotate Y=15
    Rotate Z=0
    Scale X=0
    Scale Y=0
    Scale Z=0
    Translate X=0
    Translate Y=0
    Translate Z=0
    Weight=1
   )
   > Set Transforms for Bone_3_1_1_1_3 = RigVMFunctionReferenceNode(
    Joint=(Type=Bone,Name="hair_mid_bl_mstr")
    Rotate X=0
    Rotate Y=0
    Rotate Z=15
    Scale X=0
    Scale Y=0
    Scale Z=0
    Translate X=0
    Translate Y=3
    Translate Z=0
    Weight=1
   )
   > Set Transforms for Bone_3_1_1_1_1_2 = RigVMFunctionReferenceNode(
    Joint=(Type=Bone,Name="hair_mid_br_mstr")
    Rotate X=0
    Rotate Y=0
    Rotate Z=-15
    Scale X=0
    Scale Y=0
    Scale Z=0
    Translate X=0
    Translate Y=6
    Translate Z=0
    Weight=1
   )
   > Set Transforms for Bone_3_1_1_1_2_1 = RigVMFunctionReferenceNode(
    Joint=(Type=Bone,Name="hair_mid_fl_mstr")
    Rotate X=0
    Rotate Y=0
    Rotate Z=5
    Scale X=0
    Scale Y=0
    Scale Z=0
    Translate X=0
    Translate Y=0
    Translate Z=0
    Weight=1
   )
   > Set Transforms for Bone_3_1_1_1_1_1_1 = RigVMFunctionReferenceNode(
    Joint=(Type=Bone,Name="hair_mid_fr_mstr")
    Rotate X=0
    Rotate Y=0
    Rotate Z=-5
    Scale X=0
    Scale Y=0
    Scale Z=0
    Translate X=0
    Translate Y=0
    Translate Z=0
    Weight=1
   )
  ]}
]
hierarchy [
 root {loc=<0 0 0>, rot=<0 0 0 1>}
 pelvis {parent=root, loc=<0 -0.7071 93.1191>, rot=<0.0127 0.707 -0.0127 -0.707>}
 spine_01 {parent=pelvis, loc=<6.7494 -0.6388 0>, rot=<0 0 0.034 -0.9994>}
 spine_02 {parent=spine_01, loc=<7.2518 -0.035 0>, rot=<0 0 0.0238 -0.9997>}
 spine_03 {parent=spine_02, loc=<6.0104 0 0>, rot=<0 0 0.0149 -0.9999>}
 spine_04 {parent=spine_03, loc=<8.1024 0 0>, rot=<0 0 -0.0132 -0.9999>}
 spine_05 {parent=spine_04, loc=<10.0146 0 0>, rot=<0 0 -0.0015 -1>}
 neck_01 {parent=spine_05, loc=<11.3138 0.1529 0>, rot=<0 0 -0.2003 -0.9797>}
 neck_02 {parent=neck_01, loc=<4.5627 0 0>, rot=<0 0 0.0504 -0.9987>}
 head {parent=neck_02, loc=<5.3914 0 0>, rot=<0 0 0.115 -0.9934>}
 bang_r_mstr {parent=head, loc=<11.5064 7.4904 5.2432>, rot=<0.2124 -0.0141 0.9744 -0.073>}
 bang_r_01_dyn {parent=bang_r_mstr, loc=<0 0 0>, rot=<0 0 0 1>}
 bang_r_02_dyn {parent=bang_r_01_dyn, loc=<6.6028 0 0>, rot=<-0.0105 -0.09 0.0251 -0.9956>}
 bang_r_03_dyn {parent=bang_r_02_dyn, loc=<5.5587 0 0>, rot=<-0.0056 -0.1233 0.0451 -0.9913>}
 hair_mid_fr_mstr {parent=head, loc=<9.6713 3.0428 6.8536>, rot=<0.1827 -0.0059 0.9823 -0.04>}
 hair_mid_fr_01_dyn {parent=hair_mid_fr_mstr, loc=<0 0 0>, rot=<0 0 0 1>}
 hair_mid_fr_02_dyn {parent=hair_mid_fr_01_dyn, loc=<4.8372 0 0>, rot=<-0.0066 -0.099 0.0066 -0.995>}
 hair_mid_fr_03_dyn {parent=hair_mid_fr_02_dyn, loc=<6.4225 0 0>, rot=<-0.0019 -0.0627 0.0299 -0.9976>}
 hair_mid_br_mstr {parent=head, loc=<10.2499 -2.5488 6.7605>, rot=<0.1842 -0.0081 0.9815 -0.0513>}
 hair_mid_br_01_dyn {parent=hair_mid_br_mstr, loc=<0 0 0>, rot=<0 0 0 1>}
 hair_mid_br_02_dyn {parent=hair_mid_br_01_dyn, loc=<5.4263 0 0>, rot=<-0.0106 -0.1496 0.025 -0.9884>}
 hair_mid_br_03_dyn {parent=hair_mid_br_02_dyn, loc=<6.3418 0 0>, rot=<-0.0008 -0.0349 0.0228 -0.9991>}
 hair_br_mstr {parent=head, loc=<10.2912 -6.2947 2.1284>, rot=<0.1439 -0.6918 0.6923 -0.1462>}
 hair_br_01_dyn {parent=hair_br_mstr, loc=<0 0 0>, rot=<0 0 0 1>}
 hair_br_02_dyn {parent=hair_br_01_dyn, loc=<5.4885 0 0>, rot=<0.0033 -0.1712 0.0139 -0.9851>}
 hair_br_03_dyn {parent=hair_br_02_dyn, loc=<6.3335 0 0>, rot=<-0.0004 -0.0304 -0.0127 -0.9995>}
 hair_bl_mstr {parent=head, loc=<10.3309 -6.2773 -3.3795>, rot=<0.1276 -0.6884 0.6953 -0.1621>}
 hair_bl_01_dyn {parent=hair_bl_mstr, loc=<0 0 0>, rot=<0 0 0 1>}
 hair_bl_02_dyn {parent=hair_bl_01_dyn, loc=<5.5377 0 0>, rot=<0.0001 -0.1711 0.0004 -0.9853>}
 hair_bl_03_dyn {parent=hair_bl_02_dyn, loc=<6.3391 0 0>, rot=<0.0013 -0.006 0.0245 -0.9997>}
 hair_mid_bl_mstr {parent=head, loc=<10.1744 -1.1318 -6.0539>, rot=<0.0445 -0.9805 0.0071 -0.1912>}
 hair_mid_bl_01_dyn {parent=hair_mid_bl_mstr, loc=<0 0 0>, rot=<0 0 0 1>}
 hair_mid_bl_02_dyn {parent=hair_mid_bl_01_dyn, loc=<5.3814 0 0>, rot=<0.0104 -0.1454 -0.0108 -0.9893>}
 hair_mid_bl_03_dyn {parent=hair_mid_bl_02_dyn, loc=<6.3586 0 0>, rot=<0.0014 -0.0462 -0.0302 -0.9985>}
 hair_mid_fl_mstr {parent=head, loc=<9.7136 5.1262 -6.2197>, rot=<-0.0067 -0.9815 -0.0029 -0.1913>}
 hair_mid_fl_01_dyn {parent=hair_mid_fl_mstr, loc=<0 0 0>, rot=<0 0 0 1>}
 hair_mid_fl_02_dyn {parent=hair_mid_fl_01_dyn, loc=<4.9228 0 0>, rot=<0.0032 -0.1409 0.044 -0.989>}
 hair_mid_fl_03_dyn {parent=hair_mid_fl_02_dyn, loc=<5.7533 0 0>, rot=<0.0028 -0.0511 -0.0122 -0.9986>}
 hair_bang_l_mstr {parent=head, loc=<11.9192 8.8499 -3.0856>, rot=<-0.0246 -0.9179 -0.2764 -0.2838>}
 hair_bang_l_01_dyn {parent=hair_bang_l_mstr, loc=<0 0 0>, rot=<0 0 0 1>}
 hair_bang_l_02_dyn {parent=hair_bang_l_01_dyn, loc=<2.7147 0 0>, rot=<-0.0819 -0.2782 -0.1911 -0.9377>}
 hair_bang_l_03_dyn {parent=hair_bang_l_02_dyn, loc=<4.4874 0 0>, rot=<0.016 0.0352 0.1632 -0.9858>}
 hair_bang_l_04_dyn {parent=hair_bang_l_02_dyn, loc=<3.4646 -3.1329 2.119>, rot=<0.0006 -0.0664 0.1252 -0.9899>}
 attach_cape {parent=spine_05, loc=<9.017 -11.714 0>, rot=<0.0616 -0.7044 -0.0616 -0.7044>}
 clavicle_l {parent=spine_05, loc=<4.3575 -0.1466 -2.3223>, rot=<0.0465 -0.7289 -0.009 -0.6829>}
 upperarm_l {parent=clavicle_l, loc=<15.5111 0 0>, rot=<0.0226 -0.3234 -0.0088 -0.9459>}
 lowerarm_l {parent=upperarm_l, loc=<26.6514 0 0>, rot=<0.0391 -0.0099 -0.2453 -0.9686>}
 hand_l {parent=lowerarm_l, loc=<22.3491 0 0>, rot=<-0.7081 -0.0507 0.0975 -0.6975>}
 thumb_01_l {parent=hand_l, loc=<1.3118 -0.3191 -2.459>, rot=<0.6546 -0.4042 0.0664 -0.6354>}
 thumb_02_l {parent=thumb_01_l, loc=<5.9336 -0.0448 0.0401>, rot=<-0.0037 0.0213 0.298 -0.9543>}
 thumb_03_l {parent=thumb_02_l, loc=<2.2223 0.3682 -0.0907>, rot=<0.0118 -0.013 0.0066 -0.9998>}
 index_metacarpal_l {parent=hand_l, loc=<1.6535 0.0183 -2.96>, rot=<0.0158 -0.0124 -0.0173 -0.9996>}
 index_01_l {parent=index_metacarpal_l, loc=<8.3008 0.0357 -0.0524>, rot=<-0.0044 0.0485 0.0924 -0.9945>}
 index_02_l {parent=index_01_l, loc=<3.3587 -0.3004 0.0452>, rot=<-0.002 0.0027 0.3225 -0.9466>}
 index_03_l {parent=index_02_l, loc=<2.5578 -0.123 -0.0252>, rot=<-0.002 0.0067 0.2558 -0.9667>}
 middle_metacarpal_l {parent=hand_l, loc=<2.1856 -0.1725 -0.5984>, rot=<0.006 0.0168 -0.0329 -0.9993>}
 middle_01_l {parent=middle_metacarpal_l, loc=<7.3275 -0.2447 0.0129>, rot=<-0.0208 0.0042 0.0858 -0.9961>}
 middle_02_l {parent=middle_01_l, loc=<4.0123 0.1401 0.1175>, rot=<-0.0218 0.0258 0.3512 -0.9357>}
 middle_03_l {parent=middle_02_l, loc=<2.6077 0.0534 0.1373>, rot=<0.0082 0.0188 0.1708 -0.9851>}
 ring_metacarpal_l {parent=hand_l, loc=<2.2074 -0.3625 1.3638>, rot=<-0.0284 0.0722 -0.0458 -0.9959>}
 ring_01_l {parent=ring_metacarpal_l, loc=<7.0295 -0.6197 -0.1446>, rot=<-0.0489 -0.0252 0.1892 -0.9804>}
 ring_02_l {parent=ring_01_l, loc=<3.6174 0.4647 -0.1027>, rot=<0 0 0.2806 -0.9598>}
 ring_03_l {parent=ring_02_l, loc=<2.4152 0.114 0.0199>, rot=<0 0 0.1189 -0.9929>}
 pinky_metacarpal_l {parent=hand_l, loc=<1.7616 -0.9753 3.3814>, rot=<-0.0397 0.0814 -0.069 -0.9935>}
 pinky_01_l {parent=pinky_metacarpal_l, loc=<7.1709 -0.4831 -0.1094>, rot=<-0.0654 -0.0035 0.2242 -0.9723>}
 pinky_02_l {parent=pinky_01_l, loc=<2.8274 0.0893 -0.1043>, rot=<0.0037 0.0011 0.2854 -0.9584>}
 pinky_03_l {parent=pinky_02_l, loc=<2.2046 0.0307 -0.0078>, rot=<-0.0002 0.0098 0.1445 -0.9895>}
 lowerarm_twist_02_l {parent=lowerarm_l, loc=<9.6199 0 0>, rot=<0 0 0 -1>}
 lowerarm_twist_01_l {parent=lowerarm_l, loc=<17.8549 0 0>, rot=<0 0 0 -1>}
 upperarm_twist_01_l {parent=upperarm_l, loc=<0.0041 0 0>, rot=<0 0 0 -1>}
 upperarm_twist_02_l {parent=upperarm_l, loc=<13.7941 0 0>, rot=<0 0 0 -1>}
 clavicle_r {parent=spine_05, loc=<4.3575 -0.1466 2.3223>, rot=<0.7289 0.0465 -0.6829 0.009>}
 upperarm_r {parent=clavicle_r, loc=<-15.5111 0 0>, rot=<0.0226 -0.3234 -0.0088 -0.9459>}
 lowerarm_r {parent=upperarm_r, loc=<-26.6514 0 0>, rot=<0.0391 -0.0099 -0.2453 -0.9686>}
 hand_r {parent=lowerarm_r, loc=<-22.3491 0 0>, rot=<-0.7081 -0.0507 0.0975 -0.6975>}
 thumb_01_r {parent=hand_r, loc=<-1.3118 0.3191 2.459>, rot=<0.6546 -0.4042 0.0664 -0.6354>}
 thumb_02_r {parent=thumb_01_r, loc=<-5.9336 0.0448 -0.0401>, rot=<-0.0037 0.0213 0.298 -0.9543>}
 thumb_03_r {parent=thumb_02_r, loc=<-2.2223 -0.3682 0.0907>, rot=<0.0118 -0.013 0.0066 -0.9998>}
 index_metacarpal_r {parent=hand_r, loc=<-1.6535 -0.0183 2.96>, rot=<0.0158 -0.0124 -0.0173 -0.9996>}
 index_01_r {parent=index_metacarpal_r, loc=<-8.3008 -0.0357 0.0524>, rot=<-0.0044 0.0485 0.0924 -0.9945>}
 index_02_r {parent=index_01_r, loc=<-3.3587 0.3004 -0.0452>, rot=<-0.002 0.0027 0.3225 -0.9466>}
 index_03_r {parent=index_02_r, loc=<-2.5578 0.123 0.0252>, rot=<-0.002 0.0067 0.2558 -0.9667>}
 middle_metacarpal_r {parent=hand_r, loc=<-2.1856 0.1725 0.5984>, rot=<0.006 0.0168 -0.0329 -0.9993>}
 middle_01_r {parent=middle_metacarpal_r, loc=<-7.3275 0.2447 -0.0129>, rot=<-0.0208 0.0042 0.0858 -0.9961>}
 middle_02_r {parent=middle_01_r, loc=<-4.0123 -0.1401 -0.1175>, rot=<-0.0218 0.0258 0.3512 -0.9357>}
 middle_03_r {parent=middle_02_r, loc=<-2.6077 -0.0534 -0.1373>, rot=<0.0082 0.0188 0.1708 -0.9851>}
 ring_metacarpal_r {parent=hand_r, loc=<-2.2074 0.3625 -1.3638>, rot=<-0.0284 0.0722 -0.0458 -0.9959>}
 ring_01_r {parent=ring_metacarpal_r, loc=<-7.0295 0.6197 0.1446>, rot=<-0.0489 -0.0252 0.1892 -0.9804>}
 ring_02_r {parent=ring_01_r, loc=<-3.6174 -0.4647 0.1027>, rot=<0 0 0.2806 -0.9598>}
 ring_03_r {parent=ring_02_r, loc=<-2.4152 -0.114 -0.0199>, rot=<0 0 0.1189 -0.9929>}
 pinky_metacarpal_r {parent=hand_r, loc=<-1.7616 0.9753 -3.3814>, rot=<-0.0397 0.0814 -0.069 -0.9935>}
 pinky_01_r {parent=pinky_metacarpal_r, loc=<-7.1709 0.4831 0.1094>, rot=<-0.0654 -0.0035 0.2242 -0.9723>}
 pinky_02_r {parent=pinky_01_r, loc=<-2.8274 -0.0893 0.1043>, rot=<0.0037 0.0011 0.2854 -0.9584>}
 pinky_03_r {parent=pinky_02_r, loc=<-2.2046 -0.0307 0.0078>, rot=<-0.0002 0.0098 0.1445 -0.9895>}
 lowerarm_twist_02_r {parent=lowerarm_r, loc=<-9.6199 0 0>, rot=<0 0 0 -1>}
 lowerarm_twist_01_r {parent=lowerarm_r, loc=<-17.8549 0 0>, rot=<0 0 0 -1>}
 upperarm_twist_01_r {parent=upperarm_r, loc=<-0.0041 0 0>, rot=<0 0 0 -1>}
 upperarm_twist_02_r {parent=upperarm_r, loc=<-13.7941 0 0>, rot=<0 0 0 -1>}
 thigh_l {parent=pelvis, loc=<-2.5693 -0.0109 -7.7501>, rot=<0.0012 0.0572 0.0155 -0.9982>}
 calf_l {parent=thigh_l, loc=<-39.0484 0 0>, rot=<0.0006 -0.0161 -0.0391 -0.9991>}
 foot_l {parent=calf_l, loc=<-42.561 0.1077 0.4292>, rot=<0.0418 -0.0455 0.0343 -0.9975>}
 ball_l {parent=foot_l, loc=<-6.1599 13.8264 0.6806>, rot=<0.0001 -0.0001 -0.7063 -0.7079>}
 calf_twist_02_l {parent=calf_l, loc=<-18.8064 -0.0967 0.0397>, rot=<0.0237 0.0095 -0.0035 -0.9997>}
 calf_twist_01_l {parent=calf_l, loc=<-35.813 -0.0967 0.0397>, rot=<0.0237 0.0095 -0.0035 -0.9997>}
 thigh_twist_01_l {parent=thigh_l, loc=<-21.6303 0 0>, rot=<0.0013 0 0 -1>}
 thigh_r {parent=pelvis, loc=<-2.5693 -0.0109 7.7501>, rot=<0.0572 -0.0012 0.9982 0.0155>}
 calf_r {parent=thigh_r, loc=<39.0484 0 0>, rot=<0.0006 -0.0161 -0.0391 -0.9991>}
 foot_r {parent=calf_r, loc=<42.561 -0.1077 -0.4292>, rot=<0.0418 -0.0455 0.0343 -0.9975>}
 ball_r {parent=foot_r, loc=<6.1599 -13.8264 -0.6806>, rot=<0.0001 -0.0001 -0.7063 -0.7079>}
 calf_twist_02_r {parent=calf_r, loc=<18.8064 0.0967 -0.0397>, rot=<0.0237 0.0095 -0.0035 -0.9997>}
 calf_twist_01_r {parent=calf_r, loc=<35.813 0.0967 -0.0397>, rot=<0.0237 0.0095 -0.0035 -0.9997>}
 thigh_twist_01_r {parent=thigh_r, loc=<21.6303 0 0>, rot=<0.0013 0 0 -1>}
]
